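# Physics asset for the SpringBreak keytar instrument in SparksCosmetics, defining collision bodies and joints for its dangling cloth, jaw and finger bones.
# Four four-bone finger chains (pointer, middle, ring, pinky) and a three-bone cloth chain hang from dyn_root.
# The jaw joins its own root, and the handle capsule has no constraint.
# Fingers and jaw have tight swing limits, cloth swings widely, and twist is locked on all joints except the last cloth link.
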PhysicsAsset /SparksCosmetics/Instrument/Keytar/SpringBreak/Meshes/SpringBreak_Keytar_Physics
bodies {
 cloth_dyn_01 [{shapeType=Capsule, length=1, radius=1, shapeName=cloth_dyn_01_capsule, rotation=<rot 90 0 0>}]
 cloth_dyn_02 [{shapeType=Capsule, length=0.5, radius=2, shapeName=cloth_dyn_02_capsule, rotation=<rot 90 0 0>}]
 cloth_dyn_03 [{shapeType=Capsule, length=1.5, radius=1.5, shapeName=cloth_dyn_03_capsule, center=<0.27 0 0>, rotation=<rot 90 0 0>}]
 dyn_jaw [{shapeType=Capsule, length=3.575, radius=4.425, shapeName=dyn_jaw_capsule, center=<5.1 -1.5 0>, rotation=<rot -88.5372 179.9098 -166.9829>}]
 dyn_jaw_root [{shapeType=Sphere, radius=1, shapeName=dyn_jaw_root_sphere}]
 dyn_middle_01 [{shapeType=Capsule, length=6.1691, radius=0.8896, shapeName=dyn_middle_01_capsule, center=<3.6 0 0>, rotation=<rot 85 0 0>}]
 dyn_middle_02 [{shapeType=Capsule, length=2, radius=1, shapeName=dyn_middle_02_capsule, center=<2 0 0>, rotation=<rot 85 180 180>}]
 dyn_middle_03 [{shapeType=Capsule, length=0.8211, radius=1, shapeName=dyn_middle_03_capsule, center=<1.5 0 0>, rotation=<rot 88 180 180>}]
 dyn_middle_04 [{shapeType=Capsule, length=0.3913, radius=1, shapeName=dyn_middle_04_capsule, center=<1.4 0 0>, rotation=<rot 90 180 180>}]
 dyn_pinky_01 [{shapeType=Capsule, length=4.7171, radius=0.8597, shapeName=dyn_pinky_01_capsule, center=<2.6 0 0>, rotation=<rot -84.4358 0 0>}]
 dyn_pinky_02 [{shapeType=Capsule, length=1.1958, radius=0.7908, shapeName=dyn_pinky_02_capsule, center=<1.3 0 0>, rotation=<rot -80 -180 180>}]
 dyn_pinky_03 [{shapeType=Capsule, length=0.6566, radius=0.725, shapeName=dyn_pinky_03_capsule, center=<0.7 0 0>, rotation=<rot 67 0 0>}]
 dyn_pinky_04 [{shapeType=Capsule, length=1.1434, radius=0.5746, shapeName=dyn_pinky_04_capsule, rotation=<rot 88 171 170>}]
 dyn_pointer_01 [{
   shapeType=Capsule
   length=6.0403
   radius=0.773
   shapeName=dyn_pointer_01_capsule
   center=<3.5 0 0.2>
   rotation=<rot 87.8287 -16.6482 -13.6781>
  }]
 dyn_pointer_02 [{shapeType=Capsule, length=2.1742, radius=0.8258, shapeName=dyn_pointer_02_capsule, center=<1.6 0.13 0>, rotation=<rot 90 0 0>}]
 dyn_pointer_03 [{shapeType=Capsule, length=1.6161, radius=0.6689, shapeName=dyn_pointer_03_capsule, center=<1.3 0 0>, rotation=<rot 90 0 0>}]
 dyn_pointer_04 [{shapeType=Capsule, length=0.6156, radius=0.5903, shapeName=dyn_pointer_04_capsule, center=<1 0 0>, rotation=<rot 85 180 180>}]
 dyn_ring_01 [{shapeType=Capsule, length=5.5103, radius=0.7677, shapeName=dyn_ring_01_capsule, center=<3.5 0 0>, rotation=<rot -85 0 0>}]
 dyn_ring_02 [{shapeType=Capsule, length=2, radius=1, shapeName=dyn_ring_02_capsule, center=<1.6 0 0>, rotation=<rot 85 180 180>}]
 dyn_ring_03 [{shapeType=Capsule, length=0.8409, radius=0.8738, shapeName=dyn_ring_03_capsule, center=<1.2 0 0>, rotation=<rot 90 0.3 0.33>}]
 dyn_ring_04 [{shapeType=Capsule, length=0.9957, radius=0.7314, shapeName=dyn_ring_04_capsule, center=<1 0 0>, rotation=<rot 86 0 0>}]
 dyn_root [{shapeType=Sphere, radius=1, shapeName=dyn_root_sphere}]
 handle [{shapeType=Capsule, length=19.1104, radius=1.9997, shapeName=handle_capsule, center=<0.25 0.8688 44.5404>}]
}
constraints [
 {bone1Name=cloth_dyn_01, bone2Name=dyn_root, swing1LimitDegrees=85, swing2LimitDegrees=60, twistLimitDegrees=0, twistMotion=Locked}
 {bone1Name=cloth_dyn_02, bone2Name=cloth_dyn_01, swing1LimitDegrees=70, swing2LimitDegrees=45, twistLimitDegrees=5, twistMotion=Locked}
 {bone1Name=cloth_dyn_03, bone2Name=cloth_dyn_02, swing1LimitDegrees=65, swing2LimitDegrees=45, twistLimitDegrees=5}
 {bone1Name=dyn_pointer_03, bone2Name=dyn_pointer_02, swing1LimitDegrees=8, swing2LimitDegrees=10, twistLimitDegrees=45, twistMotion=Locked}
 {bone1Name=dyn_middle_02, bone2Name=dyn_middle_01, swing1LimitDegrees=9, swing2LimitDegrees=8, twistLimitDegrees=45, twistMotion=Locked}
 {bone1Name=dyn_middle_03, bone2Name=dyn_middle_02, swing1LimitDegrees=8, swing2LimitDegrees=10, twistLimitDegrees=45, twistMotion=Locked}
 {bone1Name=dyn_middle_01, bone2Name=dyn_root, swing1LimitDegrees=4, swing2LimitDegrees=3, twistLimitDegrees=45, twistMotion=Locked}
 {bone1Name=dyn_ring_03, bone2Name=dyn_ring_02, swing1LimitDegrees=8, swing2LimitDegrees=10, twistLimitDegrees=45, twistMotion=Locked}
 {bone1Name=dyn_ring_02, bone2Name=dyn_ring_01, swing1LimitDegrees=8, swing2LimitDegrees=8, twistLimitDegrees=45, twistMotion=Locked}
 {bone1Name=dyn_ring_01, bone2Name=dyn_root, swing1LimitDegrees=3, swing2LimitDegrees=3, twistLimitDegrees=45, twistMotion=Locked}
 {bone1Name=dyn_pinky_01, bone2Name=dyn_root, swing1LimitDegrees=6, swing2LimitDegrees=8, twistLimitDegrees=45, twistMotion=Locked}
 {bone1Name=dyn_pinky_02, bone2Name=dyn_pinky_01, swing1LimitDegrees=8, swing2LimitDegrees=8, twistLimitDegrees=45, twistMotion=Locked}
 {bone1Name=dyn_pinky_03, bone2Name=dyn_pinky_02, swing1LimitDegrees=5, swing2LimitDegrees=10, twistLimitDegrees=45, twistMotion=Locked}
 {bone1Name=dyn_pinky_04, bone2Name=dyn_pinky_03, swing1LimitDegrees=8, swing2LimitDegrees=8, twistLimitDegrees=45, twistMotion=Locked}
 {bone1Name=dyn_pointer_04, bone2Name=dyn_pointer_03, swing1LimitDegrees=8, swing2LimitDegrees=8, twistLimitDegrees=45, twistMotion=Locked}
 {bone1Name=dyn_middle_04, bone2Name=dyn_middle_03, swing1LimitDegrees=8, swing2LimitDegrees=8, twistLimitDegrees=45, twistMotion=Locked}
 {bone1Name=dyn_ring_04, bone2Name=dyn_ring_03, swing1LimitDegrees=8, swing2LimitDegrees=8, twistLimitDegrees=45, twistMotion=Locked}
 {bone1Name=dyn_jaw, bone2Name=dyn_jaw_root, swing1LimitDegrees=10, swing2LimitDegrees=8, twistLimitDegrees=1, twistMotion=Locked}
 {bone1Name=dyn_pointer_01, bone2Name=dyn_root, swing1LimitDegrees=5, swing2LimitDegrees=3, twistLimitDegrees=45, twistMotion=Locked}
 {bone1Name=dyn_pointer_02, bone2Name=dyn_pointer_01, swing1LimitDegrees=8, swing2LimitDegrees=8, twistLimitDegrees=45, twistMotion=Locked}
]
